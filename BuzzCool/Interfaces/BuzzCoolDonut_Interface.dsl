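; BuzzCoolDonut_Interface is a Blueprint interface in the BuzzCool content, named for a donut.
; It declares a single function, GetSpawnActorProcessor, which has a return node and no logic of its own.
; The name suggests that implementers hand callers a processor for spawning actors, though the file shows no types or callers.
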(fn GetSpawnActorProcessor
  (return 0))
