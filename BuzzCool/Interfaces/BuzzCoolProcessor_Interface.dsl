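; Blueprint interface for a BuzzCool processor that declares how active clones are registered and queried.
; AddActiveClone takes a reference to a Wild Estate Buzz Cool pawn as the new clone, and GetActiveClones is the matching query.
; Both are declarations only, with no logic for storing or managing the clones in this file.
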(fn AddActiveClone
  (params (NewClone :type Pawn%20Wild%20Estate%20Buzz%20Cool%20Object%20Reference))
  (return))

(fn GetActiveClones
  (return 0))
